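; Creative-mode rune vent device that launches players and vehicles overlapping its capsule along their forward direction and upward, ignoring next fall damage.
; A shared tag check gates overlaps; players get launch-immunity and FX effects, with a timed retry during root motion.
; Vehicles receive a scaled ability-impact broadcast instead; in both paths, authority calls ReplicateGC for gameplay cue and trigger effects.
; SetVentActive, run on world ready and RepNotify, toggles visuals.
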(event EventOnWorldReady
  (Variables|Default|SetVentIsActive true)
  (CallFunction|SetVentActive true))

(event EventBeginPlay)

(event Custom|ReplicateGC
  (params (PlayerTarget :type Actor%20Object%20Reference))
  ;[0.2188 0.2188 0.2188] On Overlap
  ;;[1 1 1] Launch PLayer
  (CallFunction|DoLaunchEffects PlayerTarget)
  (Components|Activation|Activate (Variables|Default|GetP_Athena_RuneVent_Trigger) false))

(event OnComponentBeginOverlap(Capsule)
  (params (OverlappedComponent :type Primitive%20Component%20Object%20Reference) (OtherActor :type Actor%20Object%20Reference) (OtherComp :type Primitive%20Component%20Object%20Reference) (OtherBodyIndex :type Integer) (bFromSweep :type Boolean) (SweepResult :type Hit%20Result%20Structure%20%28by%20ref%29))
  (bind _launchcharacter (Variables|Default|GetLaunchCharacter))
  ;[0.2188 0.2188 0.2188] On Overlap
  ;;[0 1 0.02383] Overlap Happens
  (Variables|Default|SetLaunchCharacter OtherActor)
  (if (Variables|Default|GetVentIsActive)
    (bind _shouldlaunch (Class|CreativeCommonDeviceFunctionLibrary|ApplyVentRecentlyLaunchedTagOrAbortLaunch _launchcharacter self))
    (if _shouldlaunch
      (bind _asplayer_pawn_athena (Utilities|Casting|CastToPlayerPawn_Athena _launchcharacter)
        (:then
          (Variables|Default|SetPlayerTarget _asplayer_pawn_athena)
          (CallFunction|TryLaunchPlayer))
        (:CastFailed
          (CallFunction|TryLaunchVehicle))))))

(event Custom|TryLaunchPlayer
  (bind _playertarget (Variables|Default|GetPlayerTarget))
  (bind _playertarget_2 (Variables|Default|GetPlayerTarget))
  ;[0.2188 0.2188 0.2188] On Overlap
  ;;[1 1 1] Launch PLayer
  (if (GameplayTags|HasMatchingGameplayTag (Variables|Default|GetPlayerTarget) "(TagName=\"GameplayEffect.RuneVent.Applied\")")
    _
    (elif (Animation|HasAnyRootMotion (Variables|Default|GetPlayerTarget))
      ;;; HandlePlayersWIthRootMotion
      (if (not (Utilities|Time|IsTimerActivebyHandle (Variables|Default|GetRootMotionRetryTimer)))
        (bind _returnvalue (Utilities|Time|SetTimerbyEvent (EventDispatchers|CreateEvent :!event Custom|RetryRootMotionPlayer) 0.1))
        (Variables|Default|SetRootMotionRetryTimer _returnvalue))
      (else
        (Variables|/Script/FortniteGame.FortPawn|SetbIgnoreNextFallingDamage true (Variables|Default|GetPlayerTarget))
        (Utilities|FlowControl|Sequence
          (:then_0
            (Networking|SwitchHasAuthority
              (:Authority
                (Networking|FlushNetDormancy)
                (CallFunction|ReplicateGC (Variables|Default|GetPlayerTarget)))))
          (:then_1
            (Environment|ApplyGlobalEnvironmentGameplayEffecttoActor (Variables|Default|GetPlayerTarget) "/Game/Athena/Environments/Island/Runes/GE_Athena_RuneVent_LaunchImmunity.GE_Athena_RuneVent_LaunchImmunity_C" 0 "(GameplayTags=)")
            (Parachute|EndSkydiving _playertarget)
            (Environment|ApplyGlobalEnvironmentGameplayEffecttoActor _playertarget "/Game/Athena/Environments/Island/Runes/GE_Athena_RuneVent_FX.GE_Athena_RuneVent_FX_C" 0 "(GameplayTags=)")
            (bind _launchvector (CallFunction|GetLaunchVector _playertarget_2))
            (Pawn|Character|LaunchCharacterJump _playertarget_2 _launchvector true true true true)))))))

(event Custom|RetryRootMotionPlayer
  ;[0.2188 0.2188 0.2188] On Overlap
  ;;[1 1 1] Re-try if currently in root motion
  (Utilities|Time|ClearandInvalidateTimerbyHandle (Variables|Default|GetRootMotionRetryTimer))
  (for _array_element (Collision|GetOverlappingActors (Variables|Default|GetCapsule) "/Game/Athena/PlayerPawn_Athena.PlayerPawn_Athena_C" :!class /Script/Engine.PrimitiveComponent)
    (Variables|Default|SetPlayerTarget (Utilities|Casting|CastToPlayerPawn_Athena _array_element))
    (CallFunction|TryLaunchPlayer)))

(event Custom|TryLaunchVehicle
  ;[0.2188 0.2188 0.2188] On Overlap
  ;;[1 1 1] Vehicle
  (bind _asfort_vehicle_interface (Utilities|Casting|CastToFortVehicleInterface (Variables|Default|GetLaunchCharacter)))
  (bind _returnvalue (Vehicle|GetVehicleActor _asfort_vehicle_interface :!class /Script/FortniteGame.FortVehicleInterface))
  (Vehicle|SetIgnoreNextFallingDamage _asfort_vehicle_interface true)
  (bind _launchvector (CallFunction|GetLaunchVector _returnvalue))
  (Audio|BroadcastAbilityImpactatLocation 0 _returnvalue (* _launchvector 0.75) 7.8e+02)
  (Networking|SwitchHasAuthority
    (:Authority
      (CallFunction|ReplicateGC _returnvalue))))

(fn DoLaunchEffects
  (params (GCTarget :type Actor%20Object%20Reference))
  ;[1 1 1] Camera shake and Lens VFX.  Also needs SFX?
  (bind _outcueparameters (GameplayCue|BuildCueParametersfromSource GCTarget))
  (GameplayCue|ExecuteGameplayCue%28Burst%29Local GCTarget (Variables|Default|GetLaunchGameplayCue) _outcueparameters))

(fn GetLaunchVector
  (params (ActorToLaunch :type Actor%20Object%20Reference))
  (return (Math|Vector|MakeVector (.x (* (Transformation|GetActorForwardVector ActorToLaunch) (ScalableFloat|GetValueatLevel (Variables|Default|GetLaunchXY_Velocity)))) (.y (* (Transformation|GetActorForwardVector ActorToLaunch) (ScalableFloat|GetValueatLevel (Variables|Default|GetLaunchXY_Velocity)))) (ScalableFloat|GetValueatLevel (Variables|Default|GetLaunchZ_Velocity)))))

(fn IsTargetableByLinearMover
  (return true))

(fn OnRep_ForceActive)

(fn OnRep_VentIsActive
  (CallFunction|SetVentActive (Variables|Default|GetVentIsActive)))

(fn SetActiveEffects)

(fn SetVentActive
  (params (Active :type Boolean))
  (bind _pointlight (Variables|Default|GetPointLight))
  (if Active
    (Components|Activation|SetActive (Variables|Default|GetEffects_Active) true false)
    (Rendering|SetVisibility _pointlight true false)
    (Collision|SetCollisionProfileName (Variables|Default|GetCapsule) "OverlapAllDynamic" true :!class /Script/Engine.PrimitiveComponent)
    (Components|Activation|Activate (Variables|Default|GetAmbientAudio) false)
    (Rendering|Material|SetMaterial (Variables|Default|GetVent) 0 "/Game/Athena/Environments/Island/Runes/Material/MI_CorruptedVent_A_Emissive.MI_CorruptedVent_A_Emissive")
    (else
      (Components|Activation|Deactivate (Variables|Default|GetEffects_Active))
      (Rendering|SetVisibility _pointlight false false)
      (Rendering|Material|SetMaterial (Variables|Default|GetVent) 0 "/Game/Athena/Environments/Island/Runes/Material/MI_CorruptedVent_A.MI_CorruptedVent_A"))))

(fn ShouldInvisiblePropsBeAllowed
  (return false))

(fn ConstructionScript)
