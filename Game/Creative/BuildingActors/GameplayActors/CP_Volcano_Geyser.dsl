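; Creative volcano geyser prop whose cone overlap boosts player pawns and vehicles and inverts or offsets gravity on projectiles, C4 and physics balls.
; A recently-launched tag gates entry, vehicles launch only on authority, and exit suppresses fall damage.
; Server overlaps set IsActive with a five-second timeout; its OnRep drives remote material, particle, audio and timeline effects.
; Randomized materials vary each geyser, and a match-ready curve table lookup can disable it.
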(event EventBeginPlay
  (bind _s_fx_geysercone (Variables|Default|GetS_FX_GeyserCone))
  ;[1 1 1] Begin Play
  (Utilities|FlowControl|Sequence
    (:then_0
      (Rendering|Material|SetScalarParameterValueonMaterials _s_fx_geysercone "UVScale" (Math|Random|RandomFloatinRange 2.0 4.0))
      (Rendering|Material|SetScalarParameterValueonMaterials _s_fx_geysercone "SlantAmount" (Math|Random|RandomFloatinRange -5e+02 5e+02))
      (Rendering|Material|SetScalarParameterValueonMaterials _s_fx_geysercone "WPO_Amount" (Math|Random|RandomFloatinRange 0.0 5e+02))
      (Rendering|Material|SetScalarParameterValueonMaterials _s_fx_geysercone "WorldTexUVScale" (Math|Random|RandomFloatinRange 0.5 2.0))
      (bind _output_get (Variables|Default|SetGeyserContrast (Math|Random|RandomFloatinRange 1.0 3.0)))
      (Rendering|Material|SetScalarParameterValueonMaterials _s_fx_geysercone "Top Edge Contrast" _output_get)
      (Effects|Components|ParticleSystem|SetFloatParameter (Variables|Default|GetP_GeyserSteamBottom) "SpeedSpawn" (Math|Random|RandomFloatinRange 0.5 2.0)))
    (:then_1
      (if (not (Variables|Default|GetAllowEffects))
        (Components|DestroyComponent (Variables|Default|GetP_GeyserSteamBottom))))
    (:then_2
      (Class|FortAsyncActionAuthorityMatchReady|MatchReadyAsync
        (:OnReady
          (bind _outxy (DataTable|EvaluateCurveTableRow "/Game/Athena/Balance/DataTables/AthenaGameData.AthenaGameData" "Default.Vents.Disabled")
            (:RowFound
              (if (Variables|Default|GetisDisabled)
                _
                (elif (Math|Conversions|ToBoolean(Integer) (Math|Float|Truncate _outxy))
                  (Variables|Default|SetisDisabled true))))))))
    (:then_3
      (Variables|Default|SetSavedConeScale (Variables|/Script/Engine.SceneComponent|GetRelativeScale3D (Variables|Default|GetS_FX_GeyserCone))))))

(event OnComponentBeginOverlap(S_FX_GeyserCone)
  (params (OverlappedComponent :type Primitive%20Component%20Object%20Reference) (OtherActor :type Actor%20Object%20Reference) (OtherComp :type Primitive%20Component%20Object%20Reference) (OtherBodyIndex :type Integer) (bFromSweep :type Boolean) (SweepResult :type Hit%20Result%20Structure%20%28by%20ref%29))
  (bind (_returnvalue_x _returnvalue_y _returnvalue_z) (Math|Vector|BreakVector (Transformation|GetActorLocation)))
  (bind _shouldboost (Variables|Default|GetShouldBoost))
  ;[0 1 0.0429] Begin Overlap
  (bind _shouldlaunch (Class|CreativeCommonDeviceFunctionLibrary|ApplyVentRecentlyLaunchedTagOrAbortLaunch OtherActor self))
  (if _shouldlaunch
    (Utilities|FlowControl|Sequence
      (:then_0
        (bind _asfort_player_pawn (Utilities|Casting|CastToFortPlayerPawn OtherActor)
          (:then
            (Utilities|FlowControl|Sequence
              (:then_0
                (Ability|SendGameplayEventtoActor _asfort_player_pawn "(TagName=\"Abilities.Generic.Vent.Enter\")" (Utilities|Struct|MakeGameplayEventData :Target self :ContextHandle "()" :InstigatorTags "(GameplayTags=)" :TargetTags "(GameplayTags=)" :EventMagnitude (Variables|Default|GetUpwardLaunchVelocity) :TargetData "()"))
                (Development|PrintString (Utilities|String|ToString(Boolean) _shouldboost) false false)
                (if _shouldboost
                  (Character|LaunchCharacter _asfort_player_pawn (* (Transformation|GetActorUpVector) (Variables|Default|GetCharacterLaunchVelocity)) (Variables|Default|GetXYOverride) (Variables|Default|GetZOverride))
                  ;[1 1 1] Use Geysers in Different Matches Quest hook in
                  (if (Networking|IsServer)
                    (Quest|SendComplexCustomStatEvent (Quests|GetQuestManager (Fortnite|GetFortPlayerControllerfromActor _asfort_player_pawn) "Invalid") 0 "(GameplayTags=((TagName=\"Athena.Quests.Guyser\")))" "(GameplayTags=)" 1))))
              (:then_1
                (if (and (== (GameplayEffects|GetGameplayEffectCount (Ability|GetAbilitySystemComponent _asfort_player_pawn) "/Game/Athena/Environments/Blueprints/DudeBro/GE_DudeBro_Vent_Area_Effect.GE_DudeBro_Vent_Area_Effect_C" 0 true) 0) (Variables|Default|GetShouldBoost))
                  (bind (_returnvalue_x_2 _returnvalue_y_2 _returnvalue_z_2) (Math|Vector|BreakVector (Transformation|GetActorLocation _asfort_player_pawn)))
                  (Projectile|AddGameplayCue%28Looping%29Local _asfort_player_pawn (Variables|Default|GetVentCue) (Ability|GameplayCue|MakeGameplayCueParameters :NormalizedMagnitude (Math|Conversions|ToFloat(Boolean) (<= _returnvalue_z_2 (+ _returnvalue_z 8343.651367))) :RawMagnitude 0.0 :EffectContext "()" :AggregatedSourceTags "(GameplayTags=)" :AggregatedTargetTags "(GameplayTags=)" :Location "0,0,0" :Normal "0,0,0" :GameplayEffectLevel 1 :AbilityLevel 1 :bReplicateLocationWhenUsingMinimalRepProxy false))
                  (else
                    (Projectile|AddGameplayCue%28Looping%29Local _asfort_player_pawn (Variables|Default|GetVolcanoCue) (Ability|GameplayCue|MakeGameplayCueParameters :NormalizedMagnitude 0.0 :RawMagnitude 0.0 :EffectContext "()" :AggregatedSourceTags "(GameplayTags=)" :AggregatedTargetTags "(GameplayTags=)" :Location "0,0,0" :Normal "0,0,0" :GameplayEffectLevel 1 :AbilityLevel 1 :bReplicateLocationWhenUsingMinimalRepProxy false)))))))
          (:CastFailed
            (bind _asfort_vehicle_interface (Utilities|Casting|CastToFortVehicleInterface OtherActor)
              (:then
                (CallFunction|OnVehicleEnterZone (Vehicle|GetActor _asfort_vehicle_interface) -0.6)
                (Vehicle|SetIgnoreAllFallingDamage _asfort_vehicle_interface true)
                (if (Variables|Default|GetShouldBoost)
                  (Networking|SwitchHasAuthority
                    (:Authority
                      (Audio|BroadcastAbilityImpactatLocation 0 (Vehicle|GetActor _asfort_vehicle_interface) (+ (Vehicle|GetVehicleVelocity _asfort_vehicle_interface) (* (Transformation|GetActorUpVector) (Variables|Default|GetCharacterLaunchVelocity))) 777.0))) ; Only launch vehicles on the server because:\n1) we canNOT apply GEs on the client so we are unable to detect if the vehicle was launched recently and could end up launching the vehicle more times than desired\nand 2) there is an exploit where launches fired from the client are respected on the server.
                ))
              (:CastFailed
                (bind _asfort_projectile_base (Utilities|Casting|CastToFortProjectileBase OtherActor)
                  (:then
                    (CallFunction|ProjEnter _asfort_projectile_base))
                  (:CastFailed
                    (bind _asb_bga_athena_c4 (Utilities|Casting|CastToB_BGA_Athena_C4 OtherActor)
                      (:then
                        (Variables|/Script/Engine.ProjectileMovementComponent|SetProjectileGravityScale (* (Variables|/Script/Engine.ProjectileMovementComponent|GetProjectileGravityScale (Variables|/Game/Athena/Items/Consumables/C4/B_BGA_Athena_C4.B_BGA_Athena_C4_C|GetProjectileMovement1 _asb_bga_athena_c4)) -1.0) (Variables|/Game/Athena/Items/Consumables/C4/B_BGA_Athena_C4.B_BGA_Athena_C4_C|GetProjectileMovement1 _asb_bga_athena_c4)))
                      (:CastFailed
                        (bind _asphysics_ball_master (Utilities|Casting|CastToPhysicsBall_Master OtherActor))
                        (CallFunction|ToyEnter _asphysics_ball_master))))))))))
      (:then_1
        (Networking|SwitchHasAuthority
          (:Authority
            (if (Variables|Default|GetAllowEffects)
              (CallFunction|SetGeiserActive))))))))

(event OnComponentEndOverlap(S_FX_GeyserCone)
  (params (OverlappedComponent :type Primitive%20Component%20Object%20Reference) (OtherActor :type Actor%20Object%20Reference) (OtherComp :type Primitive%20Component%20Object%20Reference) (OtherBodyIndex :type Integer))
  ;[1 0.6481 0] End Overlap
  (bind _asfort_player_pawn (Utilities|Casting|CastToFortPlayerPawn OtherActor)
    (:then
      (Ability|SendGameplayEventtoActor _asfort_player_pawn "(TagName=\"Abilities.Generic.Vent.Exit\")")
      (Variables|/Script/FortniteGame.FortPawn|SetbIgnoreNextFallingDamage true _asfort_player_pawn)
      (if (Variables|Default|GetShouldBoost)
        (Projectile|RemoveGameplayCue%28Looping%29Local _asfort_player_pawn (Variables|Default|GetVentCue) (Ability|GameplayCue|MakeGameplayCueParameters :NormalizedMagnitude 0.0 :RawMagnitude 0.0 :EffectContext "()" :AggregatedSourceTags "(GameplayTags=)" :AggregatedTargetTags "(GameplayTags=)" :Location "0,0,0" :Normal "0,0,0" :GameplayEffectLevel 1 :AbilityLevel 1 :bReplicateLocationWhenUsingMinimalRepProxy false))
        (else
          (Projectile|RemoveGameplayCue%28Looping%29Local _asfort_player_pawn (Variables|Default|GetVolcanoCue) (Ability|GameplayCue|MakeGameplayCueParameters :NormalizedMagnitude 0.0 :RawMagnitude 0.0 :EffectContext "()" :AggregatedSourceTags "(GameplayTags=)" :AggregatedTargetTags "(GameplayTags=)" :Location "0,0,0" :Normal "0,0,0" :GameplayEffectLevel 1 :AbilityLevel 1 :bReplicateLocationWhenUsingMinimalRepProxy false)))))
    (:CastFailed
      (bind _asfort_vehicle_interface (Utilities|Casting|CastToFortVehicleInterface OtherActor)
        (:then
          (CallFunction|OnVehicleExitZone (Vehicle|GetActor _asfort_vehicle_interface))
          (Vehicle|SetIgnoreAllFallingDamage _asfort_vehicle_interface false)
          (Vehicle|SetIgnoreNextFallingDamage _asfort_vehicle_interface true))
        (:CastFailed
          (bind _asfort_projectile_base (Utilities|Casting|CastToFortProjectileBase OtherActor)
            (:then
              (CallFunction|ProjExit _asfort_projectile_base))
            (:CastFailed
              (bind _asb_bga_athena_c4 (Utilities|Casting|CastToB_BGA_Athena_C4 OtherActor)
                (:then
                  (Variables|/Script/Engine.ProjectileMovementComponent|SetProjectileGravityScale (* (Variables|/Script/Engine.ProjectileMovementComponent|GetProjectileGravityScale (Variables|/Game/Athena/Items/Consumables/C4/B_BGA_Athena_C4.B_BGA_Athena_C4_C|GetProjectileMovement1 _asb_bga_athena_c4)) -1.0) (Variables|/Game/Athena/Items/Consumables/C4/B_BGA_Athena_C4.B_BGA_Athena_C4_C|GetProjectileMovement1 _asb_bga_athena_c4)))
                (:CastFailed
                  (bind _asphysics_ball_master (Utilities|Casting|CastToPhysicsBall_Master OtherActor))
                  (CallFunction|ToyExit _asphysics_ball_master))))))))))

(event Custom|SetGeiserActive
  (Utilities|Time|SetTimerbyEvent (delegate Custom|GeiserTimeout) 5.0)
  (if (not (Variables|Default|GetIsActive))
    (Variables|Default|SetIsActive true)))

(event Custom|GeiserTimeout
  (Variables|Default|SetIsActive false))

(event Custom|ActiveGeyser
  (bind _s_fx_geysercone (Variables|Default|GetS_FX_GeyserCone))
  (bind _s_fx_geysercone_2 (Variables|Default|GetS_FX_GeyserCone))
  ;[1 1 1] Activate and Deactivate Geyser Visuals
  (Audio|PlaySoundatLocation (Variables|Default|GetSoundOnActivate) (Transformation|GetWorldLocation (Variables|Default|GetSM_CorruptedVent_A)))
  (Components|Activation|Activate (Variables|Default|GetP_GeyserSteam) false)
  (Rendering|Material|SetScalarParameterValueonMaterials _s_fx_geysercone "UVScale" 1.0)
  (Rendering|Material|SetScalarParameterValueonMaterials _s_fx_geysercone "WPO_Contrast" 3.0)
  (Rendering|Material|SetVectorParameterValueonMaterials _s_fx_geysercone "WorldPannerSpeed" "0.000000,2.000000,0.000000")
  (Rendering|Material|SetScalarParameterValueonMaterials _s_fx_geysercone "WPO_Amount" 1.5e+03)
  (Rendering|Material|SetScalarParameterValueonMaterials _s_fx_geysercone "WorldTexUVScale" 0.0)
  (Effects|Components|ParticleSystem|SetFloatParameter (Variables|Default|GetP_GeyserSteamBottom) "SpeedSpawn" 2.0)
  (bind (_direction _lerp_control) (Timeline|Grow Play)
    (:Update
      (Transformation|SetRelativeScale3D _s_fx_geysercone_2 (Math|Vector|Lerp(Vector) (Variables|Default|GetSavedConeScale) (+ (Variables|Default|GetSavedConeScale) "0, 0,0.050000") _lerp_control))
      (Rendering|Material|SetScalarParameterValueonMaterials _s_fx_geysercone_2 "Top Edge Contrast" (Math|Float|Lerp (Variables|Default|GetGeyserContrast) 0.5 _lerp_control))
      (Rendering|Material|SetScalarParameterValueonMaterials _s_fx_geysercone_2 "EmissiveContribution" (Math|Float|Lerp 0.1 0.3 _lerp_control))
      (Audio|Parameter|SetFloatParameter (Variables|Default|GetLoopingAmbAudio) "Intensity" _lerp_control :!class /Script/AudioExtensions.AudioParameterControllerInterface))))

(event Custom|DeactivateGeyser
  (bind _s_fx_geysercone (Variables|Default|GetS_FX_GeyserCone))
  (bind _s_fx_geysercone_2 (Variables|Default|GetS_FX_GeyserCone))
  ;[1 1 1] Activate and Deactivate Geyser Visuals
  (Audio|PlaySoundatLocation (Variables|Default|GetSoundOnDeactivate) (Transformation|GetWorldLocation (Variables|Default|GetSM_CorruptedVent_A)))
  (Components|Activation|Deactivate (Variables|Default|GetP_GeyserSteam))
  (Rendering|Material|SetScalarParameterValueonMaterials _s_fx_geysercone_2 "UVScale" (Math|Random|RandomFloatinRange 2.0 4.0))
  (Rendering|Material|SetScalarParameterValueonMaterials _s_fx_geysercone_2 "WPO_Contrast" (Math|Random|RandomFloatinRange 1.0 2.0))
  (Rendering|Material|SetScalarParameterValueonMaterials _s_fx_geysercone_2 "WPO_Amount" (Math|Random|RandomFloatinRange 0.0 5e+02))
  (Rendering|Material|SetScalarParameterValueonMaterials _s_fx_geysercone_2 "WorldTexUVScale" (Math|Random|RandomFloatinRange 0.5 2.0))
  (Variables|Default|SetGeyserContrast (Math|Random|RandomFloatinRange 1.0 3.0))
  (Effects|Components|ParticleSystem|SetFloatParameter (Variables|Default|GetP_GeyserSteamBottom) "SpeedSpawn" (Math|Random|RandomFloatinRange 0.5 2.0))
  (Timeline|Grow Reverse))

(fn GetCollisionStaticMesh
  (return "0, 0, 0" "/Game/Athena/Environments/Island/Runes/Mesh/SM_CorruptedVent_A.SM_CorruptedVent_A"))

(fn GetMeshComponents
  (return 0))

(fn IgnoreWorldGridScaleOverride
  (return false))

(fn IsTargetableByLinearMover
  (return true))

(fn OnRep_IsActive
  (Networking|SwitchHasAuthority
    (:Remote
      (if (Variables|Default|GetIsActive)
        (CallFunction|ActiveGeyser)
        (else
          (CallFunction|DeactivateGeyser))))))

(fn OnRep_isDisabled
  (if (Variables|Default|GetisDisabled)
    (Collision|SetCollisionEnabled (Variables|Default|GetS_FX_GeyserCone) "NoCollision")
    (Components|Activation|Deactivate (Variables|Default|GetP_GeyserSteamBottom))
    (Rendering|SetActorHiddenInGame self true)))

(fn ProjEnter
  (params (Projectile :type Fort%20Projectile%20Base%20Object%20Reference))
  (bind _projectilemovementcomponent (Variables|/Script/FortniteGame.FortProjectileBase|GetProjectileMovementComponent Projectile))
  (bind _projectilegravityscale (Variables|/Script/Engine.ProjectileMovementComponent|GetProjectileGravityScale _projectilemovementcomponent))
  (if (> _projectilegravityscale 0.0)
    (Variables|/Script/Engine.ProjectileMovementComponent|SetProjectileGravityScale (* _projectilegravityscale -1.0) _projectilemovementcomponent)
    (else
      (Variables|/Script/Engine.ProjectileMovementComponent|SetProjectileGravityScale (+ _projectilegravityscale (Variables|Default|GetZeroGravPrj)) _projectilemovementcomponent))))

(fn ProjExit
  (params (Projectile :type Fort%20Projectile%20Base%20Object%20Reference))
  (bind _projectilemovementcomponent (Variables|/Script/FortniteGame.FortProjectileBase|GetProjectileMovementComponent Projectile))
  (bind _projectilegravityscale (Variables|/Script/Engine.ProjectileMovementComponent|GetProjectileGravityScale _projectilemovementcomponent))
  (if (== _projectilegravityscale (Variables|Default|GetZeroGravPrj))
    (Variables|/Script/Engine.ProjectileMovementComponent|SetProjectileGravityScale (- _projectilegravityscale (Variables|Default|GetZeroGravPrj)) _projectilemovementcomponent)
    (else
      (Variables|/Script/Engine.ProjectileMovementComponent|SetProjectileGravityScale (* _projectilegravityscale -1.0) _projectilemovementcomponent))))

(fn ShouldInvisiblePropsBeAllowed
  (return false))

(fn ToyEnter
  (params (Projectile :type Physics%20Ball%20Master%20Object%20Reference))
  (bind _projectilemovement (Variables|/Game/Building/ActorBlueprints/Prop/PhysicsBall_Master.PhysicsBall_Master_C|GetProjectileMovement Projectile))
  (bind _projectilegravityscale (Variables|/Script/Engine.ProjectileMovementComponent|GetProjectileGravityScale _projectilemovement))
  (if (> _projectilegravityscale 0.0)
    (Variables|/Script/Engine.ProjectileMovementComponent|SetProjectileGravityScale (* _projectilegravityscale -0.5) _projectilemovement)
    (else
      (Variables|/Script/Engine.ProjectileMovementComponent|SetProjectileGravityScale (+ _projectilegravityscale (Variables|Default|GetZeroGravPrj)) _projectilemovement))))

(fn ToyExit
  (params (Projectile :type Physics%20Ball%20Master%20Object%20Reference))
  (bind _projectilemovement (Variables|/Game/Building/ActorBlueprints/Prop/PhysicsBall_Master.PhysicsBall_Master_C|GetProjectileMovement Projectile))
  (bind _projectilegravityscale (Variables|/Script/Engine.ProjectileMovementComponent|GetProjectileGravityScale _projectilemovement))
  (if (== _projectilegravityscale (Variables|Default|GetZeroGravPrj))
    (Variables|/Script/Engine.ProjectileMovementComponent|SetProjectileGravityScale (- _projectilegravityscale (Variables|Default|GetZeroGravPrj)) _projectilemovement)
    (else
      (Variables|/Script/Engine.ProjectileMovementComponent|SetProjectileGravityScale (* _projectilegravityscale -2.0) _projectilemovement))))

(fn ConstructionScript
  (bind _p_geysersteambottom (Variables|Default|GetP_GeyserSteamBottom))
  (|CallParentFunction)
  (Utilities|FlowControl|Sequence
    (:then_0
      (Rendering|Material|SetVectorParameterValueonMaterials (Variables|Default|GetS_FX_GeyserCone) "SulfurBaseColor" (Math|Conversions|ToVector(LinearColor) (Variables|LocalVFXColor|GetGeyser%20Illuminated%20Mesh%20Color))))
    (:then_1
      (bind _returnvalue (Rendering|Material|CreateNamedDynamicMaterialInstance _p_geysersteambottom "BaseMesh" "/Game/Effects/Fort_Effects/Materials/Water/MI_WaterFall_Basev2_GeyserBase.MI_WaterFall_Basev2_GeyserBase"))
      (Rendering|Material|SetMaterialbyName _p_geysersteambottom "BaseMesh" _returnvalue)
      (Rendering|Material|SetVectorParameterValue _returnvalue "FrenselFakeColor" (Variables|LocalVFXColor|GetBase%20Bubbling%20Mesh%20Color) :!class /Script/Engine.MaterialInstanceDynamic))))
